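# Material function for the front-end Cozy Lodge that animates a box-like, lidded prop bulging, wiggling and bursting open through World Position Offset.
# Vertex color RGB and local-position signs pick per-piece pivots, so sides fold while pieces spin, move away and shrink with ProgressFoldOpen.
# Progress_Buldge drives radial bulge and wiggle, with a directional shake added last.
# Static switches provide time-looped previews or swap the explosion for a late collapse to the object position.
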MaterialFunction /Game/Environments/FrontEnd/MaterialFunctions/MF_CozyLodge_Anim
graph {nodes [
  n1 = GetMaterialAttributes(None=FunctionInput(Preview=SetMaterialAttributes!(MaterialAttributes)) {bUsePreviewValueAsDefault=true, inputType=FunctionInput_MaterialAttributes}) {attributeGetTypes [#1]}
  n2 = Transform(None=Subtract(A=WorldPosition().XYZ, B=ObjectPositionWS())) {transformSourceType=TRANSFORMSOURCE_World, transformType=TRANSFORM_Local}
  n3 = ComponentMask(None=@n2) {g=true}
  n4 = Saturate(None=@n3)
  n5 = LinearInterpolate(
   A=VectorParameter() {parameterName=PivotPosXL, defaultValue=<rgba 0 -28 0 1>}.RGB
   B=VectorParameter() {parameterName=PivotPosXR, defaultValue=<rgba 0 28 0 1>}.RGB
   Alpha=@n4
  )
  n6 = VertexColor()
  n7 = LinearInterpolate(A=VectorParameter() {parameterName=PivotPosBottom, defaultValue=<rgba 0 0 0 1>}.RGB, B=@n5, Alpha=@n6.G)
  n8 = ComponentMask(None=@n2) {r=true}
  n9 = Saturate(None=@n8)
  n10 = LinearInterpolate(
   A=VectorParameter() {parameterName=PivotPosYL, defaultValue=<rgba -28 0 0 1>}.RGB
   B=VectorParameter() {parameterName=PivotPosYR, defaultValue=<rgba 28 0 0 1>}.RGB
   Alpha=@n9
  )
  n11 = TransformPosition(None=LinearInterpolate(
    A=LinearInterpolate(A=@n7, B=@n10, Alpha=@n6.R)
    B=VectorParameter() {parameterName=PivotPosLid, defaultValue=<rgba 0 0 100 1>}.RGB
    Alpha=@n6.B
   )) {transformType=TRANSFORMPOSSOURCE_World}
  n12 = Time()
  n13 = ScalarParameter() {defaultValue=2.0231, parameterName=Speed}
  n14 = StaticSwitchParameter(True=Frac(None=Multiply(A=@n12, B=@n13)), False=ScalarParameter() {parameterName=ProgressFoldOpen}) {parameterName=DebugAnim}
  n15 = Saturate(None=Power(Base=@n14, Exp=ScalarParameter() {defaultValue=4.1285, parameterName=ShrinkPieces}))
  n16 = Multiply(A=Subtract(A=@n11, B=WorldPosition().XYZ), B=@n15)
  n17 = Transform(None=VectorParameter() {parameterName=XAxis, defaultValue=<rgba -1 0 0 1>}.RGB) {transformSourceType=TRANSFORMSOURCE_Local}
  n18 = Multiply(A=ScalarParameter() {defaultValue=0.25, parameterName=RotOpen}, B=Sign(None=@n3))
  n19 = Add(A=WorldPosition().XYZ, B=@n16)
  n20 = Multiply(
   A=VertexColor().G
   B=RotateAboutAxis(
    NormalizedRotationAxis=Normalize(VectorInput=@n17)
    RotationAngle=Multiply(A=@n14, B=@n18)
    PivotPoint=TransformPosition(None=@n5) {transformType=TRANSFORMPOSSOURCE_World}
    Position=@n19
   )
  )
  n21 = Normalize(VectorInput=Transform(None=VectorParameter() {parameterName=YAxis, defaultValue=<rgba 0 1 0 1>}.RGB) {transformSourceType=TRANSFORMSOURCE_Local})
  n22 = Multiply(A=@n14, B=Multiply(A=ScalarParameter() {defaultValue=0.25, parameterName=RotOpen}, B=Sign(None=@n8)))
  n23 = TransformPosition(None=@n10) {transformType=TRANSFORMPOSSOURCE_World}
  n24 = VectorParameter() {parameterName=RotAxisSpin, defaultValue=<rgba 1 1 0 1>}
  n25 = ScalarParameter() {defaultValue=0.854, parameterName=Spin}
  n26 = Saturate(None=Power(Base=@n14, Exp=ScalarParameter() {defaultValue=1.4475, parameterName=Acceleration}))
  n27 = VectorParameter() {parameterName=Move To Location, defaultValue=<rgba 0 0 -200 1>}
  n28 = LinearInterpolate(
   A=VectorParameter() {parameterName=Move To Location_XL, defaultValue=<rgba 0 200 -100 1>}.RGB
   B=VectorParameter() {parameterName=Move To Location_XR, defaultValue=<rgba 0 -200 -100 1>}.RGB
   Alpha=@n4
  )
  n29 = VectorParameter() {parameterName=Move To Location_YL, defaultValue=<rgba 200 0 -100 1>}
  n30 = VectorParameter() {parameterName=Move To Location_YR, defaultValue=<rgba -200 0 -100 1>}
  n31 = LinearInterpolate(
   A=LinearInterpolate(A=LinearInterpolate(A=@n27.RGB, B=@n28, Alpha=@n6.G), B=LinearInterpolate(A=@n29.RGB, B=@n30.RGB, Alpha=@n9), Alpha=@n6.R)
   B=VectorParameter() {parameterName=Move To Location_Lid, defaultValue=<rgba 0 0 324.3209 1>}.RGB
   Alpha=@n6.B
  )
  n32 = Power(Base=@n14, Exp=ScalarParameter() {defaultValue=0.965, parameterName=AccelerationMovement})
  n33 = Add(
   A=RotateAboutAxis(NormalizedRotationAxis=Normalize(VectorInput=@n24.RGB), RotationAngle=LinearInterpolate(B=@n25, Alpha=@n26), PivotPoint=@n11, Position=@n19)
   B=LinearInterpolate(A=Constant3Vector(), B=Transform(None=@n31) {transformSourceType=TRANSFORMSOURCE_Local}, Alpha=Saturate(None=@n32))
  )
  n34 = Add(A=Multiply(A=VertexColor().R, B=RotateAboutAxis(NormalizedRotationAxis=@n21, RotationAngle=@n22, PivotPoint=@n23, Position=@n19)), B=@n33)
  n35 = Add(A=@n1.WorldPositionOffset, B=WorldPosition().XYZ)
  n36 = ObjectPositionWS()
  n37 = Time()
  n38 = ScalarParameter() {defaultValue=0.1, parameterName=EmissiveDebugSpeed}
  n39 = StaticSwitchParameter(True=Frac(None=Multiply(A=@n37, B=@n38)), False=ScalarParameter() {parameterName=Progress_Buldge}) {parameterName=DebugEmissiveAnim}
  n40 = MaterialFunctionCall() {materialFunction=/Engine/Functions/Engine_MaterialFunctions02/WorldPositionOffset/ObjectScale}
  n41 = ScalarParameter() {defaultValue=48.3897, parameterName=Push}
  n42 = Multiply(A=Normalize(VectorInput=Subtract(A=@n35, B=@n36)), B=Multiply(A=@n39, B=Multiply(A=@n40.Scale X, B=@n41)))
  n43 = Distance(A=@n35, B=@n36)
  n44 = Multiply(A=@n40.Scale X, B=ScalarParameter() {defaultValue=46.5695, parameterName=PushD})
  n45 = Power(Base=OneMinus(None=Divide(A=@n43, B=@n44)), Exp=ScalarParameter() {defaultValue=0.966, parameterName=PushFalloff})
  n46 = Time()
  n47 = Sine(None=Multiply(A=@n46, B=ScalarParameter() {defaultValue=3, parameterName=WiggleSpped}))
  n48 = LinearInterpolate(A=ScalarParameter() {parameterName=Rotation_1}, B=ScalarParameter() {defaultValue=0.02, parameterName=Rotation_2}, Alpha=@n39)
  n49 = TransformPosition(None=VectorParameter() {parameterName=PivotPosition, defaultValue=<rgba 0 0 40.878 1>}.RGB) {transformType=TRANSFORMPOSSOURCE_World}
  n50 = Add(A=@n1.WorldPositionOffset, B=WorldPosition().XYZ)
  n51 = RotateAboutAxis(
   NormalizedRotationAxis=Normalize(VectorInput=VectorParameter() {parameterName=RotationAxis, defaultValue=<rgba 1 0 -0.1524 1>}.RGB)
   RotationAngle=Multiply(A=Multiply(A=@n47, B=@n48), B=ScalarParameter() {defaultValue=1, parameterName=Progress_Reverse_Blocker})
   PivotPoint=@n49
   Position=@n50
  )
  n52 = Multiply(A=@n46, B=ScalarParameter() {defaultValue=4, parameterName=WiggleSpped2})
  n53 = RotateAboutAxis(
   NormalizedRotationAxis=Normalize(VectorInput=VectorParameter() {parameterName=RotationAxis2, defaultValue=<rgba 0 1 0.2095 1>}.RGB)
   RotationAngle=Multiply(A=@n48, B=Sine(None=@n52))
   PivotPoint=@n49
   Position=Add(A=@n51, B=@n50)
  )
  n54 = Add(A=@n1.WorldPositionOffset, B=Add(A=Multiply(A=@n42, B=@n45), B=Add(A=@n51, B=@n53)))
  n55 = Add(A=@n16, B=Add(A=Add(A=@n20, B=@n34), B=@n54))
  n56 = Constant()
  n57 = LinearInterpolate(
   A=@n54
   B=Subtract(A=ObjectPositionWS(), B=WorldPosition().XYZ)
   Alpha=If(A=@n14, B=Constant() {r=0.95}, "A > B"=Constant() {r=1}, "A == B"=@n56, "A < B"=@n56)
  )
  n58 = GetMaterialAttributes(None=SetMaterialAttributes!(
    MaterialAttributes=@n1.MaterialAttributes
    WorldPositionOffset=StaticSwitchParameter(True=@n55, False=@n57) {defaultValue=true, parameterName=UseExplodeAnim}
   ) {attributeSetTypes [#1]}) {attributeGetTypes [#1]}
  n59 = Normalize(VectorInput=Transform(None=VectorParameter() {parameterName=shakedirection, defaultValue=<rgba 0 0 1 1>}.RGB) {transformSourceType=TRANSFORMSOURCE_Local})
  n60 = Sine(None=Multiply(A=Time(), B=ScalarParameter() {defaultValue=3, parameterName=ShakeSpeed}))
  n61 = Multiply(
   A=Multiply(A=Multiply(A=@n59, B=@n60), B=ScalarParameter() {defaultValue=3, parameterName=ShakeDist})
   B=ScalarParameter() {parameterName=ShakeAmount}
  )
  FunctionOutput(None=SetMaterialAttributes(MaterialAttributes=@n58.MaterialAttributes, WorldPositionOffset=Add(A=@n58.WorldPositionOffset, B=@n61)) {attributeSetTypes [#1]})
  Multiply(A=VertexColor().B, B=@n33)
  OneMinus(None=@n15)
  WorldPosition()
  WorldPosition()
 ]}
